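# Niagara effect for the gallon bag backpack cosmetic that sheds small, short-lived cookie crumb mesh particles.
# A single emitter drives spawn rate from a curve over normalized loop age, with a sharp early spike, a lull and a second rise.
# Crumbs spawn in an offset, stretched sphere with cone velocity, then fall under randomized gravity and drag while scaling up and shrinking away.
# Front-end and hit-vector user parameters are declared but not referenced by the shown modules.
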NiagaraSystem /BRCosmetics/Accessories/FORT_Backpacks/Backpack_M_MED_GallonBag/FX/NS_Backpack_GallonBag_CookieCrumbs
dependencies {
 usedDataInterfaces [&NiagaraDataInterfaceCurve, &NiagaraDataInterfaceMeshRendererInfo]
 usedDynamicInputs []
 usedModules [&SystemState]
}
emitters {Cookie_Crumbs {
  emitterUpdateScript [
   &EmitterState()
   &SpawnRate(
    SpawnRate=&FloatFromCurve(
     FloatCurve=DataInterface() {propertyValues {Curve {keys [
         {arriveTangent=24.0596, interpMode=RCIM_Cubic, leaveTangent=24.0596, tangentMode=RCTM_User, time=0, value=0}
         {arriveTangent=-3.432, interpMode=RCIM_Cubic, leaveTangent=-3.432, tangentMode=RCTM_User, time=0.1, value=1}
         {arriveTangent=-0.0004, interpMode=RCIM_Cubic, leaveTangent=-0.0004, tangentMode=RCTM_User, time=0.4, value=0.05}
         {arriveTangent=-0.0212, interpMode=RCIM_Cubic, leaveTangent=-0.0212, tangentMode=RCTM_User, time=0.6, value=0.05}
         {arriveTangent=-0.125, interpMode=RCIM_Cubic, leaveTangent=-0.125, time=0.95, value=0.6}
         {arriveTangent=-37.3311, tangentMode=RCTM_User, time=1, value=0}
        ]}}}
     CurveIndex=$Emitter.NormalizedLoopAge
     Scale Curve=40
    )
    Use Spawn Probability=true
   )
  ]
  particleSpawnScript [
   &InitializeParticle(
    Color=<rgba 0.625 2.3437 2.5 1>
    Lifetime Max=0.725
    Lifetime Min=0.35
    Lifetime Mode=::NewEnumerator1
    Mesh Renderer Array Visibility Mode=::NewEnumerator3
    Mesh Renderer Info=DataInterface() {propertyValues {MeshRenderer=&self:Cookie_Crumbs_0.NiagaraMeshRendererProperties_0}}
    Mesh Scale Mode=::NewEnumerator1
    Mesh Uniform Scale Max=0.06
    Mesh Uniform Scale Min=0.04
    Position Offset=<0 -1.25 0.5>
    Sprite Rotation Mode=::NewEnumerator2
    UsePositionOffset=true
    Write Material Random=true
   )
   &CameraOffset(Camera Offset Amount=1)
   &AddVelocity(
    Cone Angle=90
    Cone Axis=<1 0.2 -0.2>
    Module.UseConeFalloff=false
    Module.UseVelDistribution=false
    Override Seed=false
    Velocity Mode=::NewEnumerator2
    Velocity Speed=&RandomRangeFloat(Minimum=5, Maximum=10)
   )
   &ShapeLocation(
    MaskSpawnGroup=false
    Non Uniform Scale=<0.2 0.5 2>
    Override Seed=false
    Rotation Axis=<0 0.3 1>
    Rotation Mode=::NewEnumerator1
    Sphere Radius=1
    UseOwnerScale=true
   )
   &InitialMeshOrientation(Use Rotation Vector=false)
  ]
  particleUpdateScript [
   &ParticleState()
   &ScaleMeshSize(Scale Factor=&VectorFromFloat(Value=&FloatFromCurve(FloatCurve=DataInterface() {propertyValues {Curve {keys [
          {arriveTangent=1.6129, interpMode=RCIM_Cubic, leaveTangent=1.6129, tangentMode=RCTM_User, time=0, value=0.65}
          {interpMode=RCIM_Cubic, tangentMode=RCTM_User, time=0.5, value=1}
          {arriveTangent=-2.9827, interpMode=RCIM_Cubic, leaveTangent=-2.9827, tangentMode=RCTM_User, time=1, value=0}
         ]}}})))
   &GravityForce(Gravity=&MakeVector(X=0, Y=0, Z=&RandomRangeFloat(Minimum=-40, Maximum=-60)))
   &Drag(Drag=2, Use Linear Drag=true, Use Rotational Drag=true)
   &SolveForcesAndVelocity(Clamp Velocity=false, Limit Acceleration=false)
  ]
  rendererClasses [&NiagaraMeshRendererProperties]
  renderers [{rendererClass=&NiagaraMeshRendererProperties, rendererIndex=0}]
 }}
summary {
 systemName=NS_Backpack_GallonBag_CookieCrumbs
 emitters [{emitterName=Cookie_Crumbs, rendererClasses [&NiagaraMeshRendererProperties]}]
 userVariables [
  User.bIsFrontEnd {type=NiagaraBool}
  User.bIsFrontEndPreview {type=NiagaraBool}
  User.FrontendScaler {type=NiagaraFloat, defaultValue=0.555, description=""}
  User.Hit Vector {type=Vector3f, description="", defaultValue {struct=Vector3f, value=<1 0 0>}}
 ]
}
