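# Character material for the Gallon Bag backpack that adds a scalar-driven "puke" reaction look on top of the 2023 character parent.
# Between the parent's PreFX and PostFX functions, it blends base color toward PukeColor, scales roughness, and layers in a flattenable puke normal via angle-corrected blending.
# The Puke_Curve parameter drives the effect through named reroutes: a saturated alpha and an area alpha masked by a selectable Tech Art Mask channel, green by default.
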Material /BRCosmetics/Accessories/FORT_Backpacks/Backpack_M_MED_GallonBag/Materials/M_Backpack_GallonBag_Reactivity
graph {
 properties {
  bUseMaterialAttributes=true
  bUsedWithMeshDeformer=true
  bUsedWithSkeletalMesh=true
  pixelDepthOffsetMode=PDOM_Legacy
  shadingModel=MSM_FromMaterialExpression
  usedShadingModels=Default Lit | Subsurface | Clear Coat | Subsurface Profile | Cloth
 }
 nodes [
  n1 = MaterialFunctionCall() {materialFunction=&MF_CharacterParent_2023_PreFX}
  n2 = GetMaterialAttributes(None=@n1.Result) {attributeGetTypes [#1, #2, #3]}
  n3 = VectorParameter() {group=Puke_React, parameterName=PukeColor, defaultValue=<rgba 1 1 1 1>}
  n4 = NamedRerouteUsage()
  n5 = Multiply(A=@n2.Roughness, B=ScalarParameter() {defaultValue=0.2, group=Puke_React, parameterName=RoughnessOverride})
  n6 = NamedRerouteUsage()
  n7 = TextureSampleParameter2D() {
   group=Puke_React
   parameterName=PukeEffectNormal
   samplerType=SAMPLERTYPE_Normal
   texture=/BRCosmetics/Accessories/FORT_Backpacks/Backpack_M_MED_GallonBag/Textures/T_Backpack_GallonBag_Puke_N
  }
  n8 = MaterialFunctionCall(Normal=@n7.RGB, Flatness=OneMinus(None=NamedRerouteUsage())) {materialFunction=&FlattenNormal}
  MaterialFunctionCall(
   Material=SetMaterialAttributes!(
    MaterialAttributes=@n2.MaterialAttributes
    Base Color=LinearInterpolate(A=@n2.Base Color, B=@n3.RGB, Alpha=@n4)
    Roughness=LinearInterpolate(A=@n2.Roughness, B=@n5, Alpha=@n6)
    Normal=MaterialFunctionCall(BaseNormal=@n2.Normal, AdditionalNormal=@n8.Result) {materialFunction=&BlendAngleCorrectedNormals}.Result
   ) {attributeSetTypes [#1, #2, #3]}
   M=@n1.M
   Pre Skinned Local Position=@n1.M
   Distance Blend=@n1.M
   Fresnel=@n1.M
  ) {materialFunction=&MF_CharacterParent_2023_PostFX}
  n9 = ScalarParameter() {group=Puke_React, parameterName=Puke_Curve, sliderMax=1}
  NamedRerouteDeclaration(None=Multiply(
    A=ChannelMaskParameter(None=@n1.Tech Art Mask) {group=Puke_React, maskChannel=Green, parameterName=PukeEffectChannel, defaultValue=<rgba 0 1 0 0>}.R
    B=@n9
   )) {name=AlphaArea, nodeColor=<rgba 1 0.6328 0 1>}
  NamedRerouteDeclaration(None=Saturate(None=@n9)) {name=Alpha, nodeColor=<rgba 0.3125 0 1 1>}
 ]
}
