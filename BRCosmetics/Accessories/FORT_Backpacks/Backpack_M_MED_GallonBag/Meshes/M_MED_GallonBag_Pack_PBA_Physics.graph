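# Physics asset for the GallonBag backpack cosmetic, giving collision bodies and joint limits to a bag and a small eared character rig.
# The pack body, a box plus a tiny capsule, anchors base_bag, which carries the bag_mid to bag_lower chain and base_body under tight angular limits.
# From base_body, capsule chains cover spine, arms, neck, head and two-segment ears, with looser uniform limits that let the character flop more than the bag.
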PhysicsAsset /BRCosmetics/Accessories/FORT_Backpacks/Backpack_M_MED_GallonBag/Meshes/M_MED_GallonBag_Pack_PBA_Physics
bodies {
 bag_lower [{shapeType=Capsule, length=13.6734, radius=9.4663, shapeName=bag_lower_capsule, center=<5 0 -3>, rotation=<rot 0 0 90>}]
 bag_mid [{shapeType=Capsule, length=11.4606, radius=10.6425, shapeName=bag_mid_capsule, center=<5 0 -3>, rotation=<rot 0 0 90>}]
 base_bag [{shapeType=Capsule, length=14.2942, radius=8.308, shapeName=base_bag_capsule, center=<9.5958 -0.2501 -4.2013>, rotation=<rot 18.2513 0 -90>}]
 base_body [{
   shapeType=Capsule
   length=2.2597
   radius=5.6778
   shapeName=base_body_capsule
   center=<1.1919 0.0112 -0.7328>
   rotation=<rot 0 179.4304 -89.211>
  }]
 pack [
  {shapeType=Capsule, length=0.505, radius=0.505, shapeName=pack_capsule, rotation=<rot 0 -90 0>}
  {shapeType=Box, extentX=22.1875, extentY=4.3125, extentZ=13.4375, shapeName=pack_box, center=<0 0 7.854>}
 ]
 pack_clavicle_l [{
   shapeType=Capsule
   length=4.0862
   radius=1.4794
   shapeName=clavicle_l_capsule
   center=<3.5778 1.0662 0.0456>
   rotation=<rot 0 -4.7265 -81.233>
  }]
 pack_clavicle_r [{
   shapeType=Capsule
   length=4.4279
   radius=1.5813
   shapeName=clavicle_r_capsule
   center=<3.7145 -1.3347 0.0646>
   rotation=<rot 0 -175.169 -78.9325>
  }]
 pack_earTip_l [{shapeType=Capsule, length=2.4681, radius=1.6837, shapeName=earTip_l_capsule, center=<2.2376 0 0>, rotation=<rot 0 90 90>}]
 pack_earTip_r [{shapeType=Capsule, length=2.4681, radius=1.6842, shapeName=earTip_r_capsule, center=<2.2086 0 0>, rotation=<rot 0 -90 -90>}]
 pack_ear_l [{shapeType=Capsule, length=3.5477, radius=2.9265, shapeName=ear_l_capsule, center=<2.4909 0 0>, rotation=<rot 0 90 90>}]
 pack_ear_r [{shapeType=Capsule, length=3.5477, radius=2.9295, shapeName=ear_r_capsule, center=<2.4593 0 0>, rotation=<rot 0 -90 -90>}]
 pack_hand_l [{shapeType=Capsule, length=2.1764, radius=1.4557, shapeName=hand_l_capsule, center=<0.4075 0.0669 -0.3199>, rotation=<rot 0 92.8561 -88.2726>}]
 pack_hand_r [{shapeType=Capsule, length=2.1764, radius=1.4557, shapeName=hand_l_capsule, center=<0.4075 -0.0669 0.3199>, rotation=<rot 0 92.8561 91.7274>}]
 pack_head [{shapeType=Capsule, length=2.5612, radius=5.2336, shapeName=head_capsule, center=<2.9774 0.0824 2.6533>, rotation=<rot -2.1637 -90 -14.4337>}]
 pack_lowerarm_l [{shapeType=Capsule, length=1.5625, radius=2.5, shapeName=lowerarm_l_capsule, center=<1.557 0 -0.7727>, rotation=<rot 90 0 0>}]
 pack_lowerarm_r [{shapeType=Capsule, length=1.5625, radius=2.5, shapeName=lowerarm_l_capsule, center=<1.557 0 -0.7727>, rotation=<rot 90 0 180>}]
 pack_neck_01 [{shapeType=Capsule, length=0.505, radius=0.505, shapeName=neck_01_capsule, center=<-0.0495 -0.0004 0.0258>, rotation=<rot 0 -90 -1.7573>}]
 pack_neck_02 [{shapeType=Capsule, length=0.505, radius=0.505, shapeName=neck_02_capsule, center=<-0.0588 -0.0004 -0.0059>, rotation=<rot 0 -90 -22.6457>}]
 pack_spine_01 [{shapeType=Capsule, length=6.2757, radius=5.7576, shapeName=spine_01_capsule, rotation=<rot 0 0 90>}]
 pack_spine_02 [{shapeType=Capsule, length=4.7812, radius=4.7863, shapeName=spine_02_capsule, center=<0 0 -0.5274>, rotation=<rot 0 0 90>}]
 pack_spine_03 [{shapeType=Capsule, length=2.8402, radius=3.927, shapeName=spine_03_capsule, center=<0 0 -1.7937>}]
 pack_upperarm_l [{shapeType=Capsule, length=2.3313, radius=1.9094, shapeName=upperarm_l_capsule, center=<0.5782 -0.9205 -0.3343>, rotation=<rot 90 0 0>}]
 pack_upperarm_r [{shapeType=Capsule, length=2.3313, radius=1.9094, shapeName=upperarm_l_capsule, center=<0.5782 0.9205 -0.3343>, rotation=<rot 90 0 180>}]
}
constraints [
 {bone1Name=base_body, bone2Name=base_bag, swing1LimitDegrees=1, swing2LimitDegrees=1, twistLimitDegrees=1}
 {bone1Name=base_bag, bone2Name=pack, swing1LimitDegrees=5, swing2LimitDegrees=5, twistLimitDegrees=5}
 {bone1Name=bag_mid, bone2Name=base_bag, swing1LimitDegrees=9, swing2LimitDegrees=2, twistLimitDegrees=9}
 {bone1Name=bag_lower, bone2Name=bag_mid, swing1LimitDegrees=9, swing2LimitDegrees=2, twistLimitDegrees=9}
 {bone1Name=pack_spine_01, bone2Name=base_body, swing1LimitDegrees=45, swing2LimitDegrees=45, twistLimitDegrees=45}
 {bone1Name=pack_spine_02, bone2Name=pack_spine_01, swing1LimitDegrees=45, swing2LimitDegrees=45, twistLimitDegrees=45}
 {bone1Name=pack_spine_03, bone2Name=pack_spine_02, swing1LimitDegrees=45, swing2LimitDegrees=45, twistLimitDegrees=45}
 {bone1Name=pack_clavicle_l, bone2Name=pack_spine_03, swing1LimitDegrees=45, swing2LimitDegrees=45, twistLimitDegrees=45}
 {bone1Name=pack_clavicle_r, bone2Name=pack_spine_03, swing1LimitDegrees=45, swing2LimitDegrees=45, twistLimitDegrees=45}
 {bone1Name=pack_neck_01, bone2Name=pack_spine_03, swing1LimitDegrees=45, swing2LimitDegrees=45, twistLimitDegrees=45}
 {bone1Name=pack_upperarm_l, bone2Name=pack_clavicle_l, swing1LimitDegrees=45, swing2LimitDegrees=45, twistLimitDegrees=45}
 {bone1Name=pack_lowerarm_l, bone2Name=pack_upperarm_l, swing1LimitDegrees=45, swing2LimitDegrees=45, twistLimitDegrees=45}
 {bone1Name=pack_hand_l, bone2Name=pack_lowerarm_l, swing1LimitDegrees=45, swing2LimitDegrees=45, twistLimitDegrees=45}
 {bone1Name=pack_upperarm_r, bone2Name=pack_clavicle_r, swing1LimitDegrees=45, swing2LimitDegrees=45, twistLimitDegrees=45}
 {bone1Name=pack_lowerarm_r, bone2Name=pack_upperarm_r, swing1LimitDegrees=45, swing2LimitDegrees=45, twistLimitDegrees=45}
 {bone1Name=pack_hand_r, bone2Name=pack_lowerarm_r, swing1LimitDegrees=45, swing2LimitDegrees=45, twistLimitDegrees=45}
 {bone1Name=pack_neck_02, bone2Name=pack_neck_01, swing1LimitDegrees=45, swing2LimitDegrees=45, twistLimitDegrees=45}
 {bone1Name=pack_head, bone2Name=pack_neck_02, swing1LimitDegrees=45, swing2LimitDegrees=45, twistLimitDegrees=45}
 {bone1Name=pack_ear_l, bone2Name=pack_head, swing1LimitDegrees=45, swing2LimitDegrees=45, twistLimitDegrees=45}
 {bone1Name=pack_ear_r, bone2Name=pack_head, swing1LimitDegrees=45, swing2LimitDegrees=45, twistLimitDegrees=45}
 {bone1Name=pack_earTip_l, bone2Name=pack_ear_l, swing1LimitDegrees=45, swing2LimitDegrees=45, twistLimitDegrees=45}
 {bone1Name=pack_earTip_r, bone2Name=pack_ear_r, swing1LimitDegrees=45, swing2LimitDegrees=45, twistLimitDegrees=45}
]
