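# Physics asset for the M_MED_GallonBag backpack cosmetic, giving capsule collision and joints to the bag, its front straps and a cookie bag.
# Two capsules form the base bag, from which two front-strap chains and a pack_hand_l to cookie_bag chain branch.
# Constraints at the chain roots and hand-to-cookie link swing and twist freely.
# The dyn strap tips have asymmetric swing limits with minimal twist, while cookie_bag_01 is tightly limited against cookie_bag.
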PhysicsAsset /BRCosmetics/Accessories/FORT_Backpacks/Backpack_M_MED_GallonBag/Meshes/M_MED_GallonBag_Pack_Physics
bodies {
 base_bag [
  {
   shapeType=Capsule
   length=11.5887
   radius=9.5009
   shapeName=base_bag_capsule
   center=<6.3734 0.0015 -13.8416>
   rotation=<rot 14.6509 -1.312 89.6681>
  }
  {
   shapeType=Capsule
   length=11.9688
   radius=8.7188
   shapeName=base_bag_capsule
   center=<4.6215 -0.035 -0.4542>
   rotation=<rot 21.0521 0 -89.0388>
  }
 ]
 base_bag_front_strap_l [{shapeType=Capsule, length=0, radius=1, shapeName=base_bag_front_strap_l_capsule, rotation=<rot 90 0 0>}]
 base_bag_front_strap_r [{shapeType=Capsule, length=0, radius=1, shapeName=base_bag_front_strap_r_capsule, rotation=<rot 90 0 0>}]
 cookie_bag [{shapeType=Capsule, length=2.5374, radius=2.4048, center=<2.2125 0.0389 -1.7179>, rotation=<rot 0 -130.0257 -84.7086>}]
 cookie_bag_01 [{
   shapeType=Capsule
   length=2.1038
   radius=2.4062
   shapeName=cookie_bag_01_capsule
   center=<0.6893 -0.0459 -0.1204>
   rotation=<rot -20.2038 -89.1383 -90.0128>
  }]
 dyn_bag_front_strap_l [{shapeType=Capsule, length=2.7471, radius=1.1354, shapeName=dyn_bag_front_strap_l_capsule, center=<3.5458 0 0>, rotation=<rot 90 0 0>}]
 dyn_bag_front_strap_r [{shapeType=Capsule, length=4.8698, radius=1.1354, shapeName=dyn_bag_front_strap_r_capsule, center=<4.5789 0 0>, rotation=<rot 90 0 0>}]
 pack_hand_l [{
   shapeType=Capsule
   length=2.1764
   radius=1.4557
   shapeName=pack_hand_l_capsule
   center=<0.4075 0.0669 -0.3199>
   rotation=<rot 0 92.8561 -88.2726>
  }]
}
constraints [
 {
  bone1Name=base_bag_front_strap_l
  bone2Name=base_bag
  swing1LimitDegrees=45
  swing1Motion=Free
  swing2LimitDegrees=45
  swing2Motion=Free
  twistLimitDegrees=45
  twistMotion=Free
 }
 {bone1Name=dyn_bag_front_strap_l, bone2Name=base_bag_front_strap_l, swing1LimitDegrees=55, swing2LimitDegrees=85, twistLimitDegrees=5}
 {
  bone1Name=base_bag_front_strap_r
  bone2Name=base_bag
  swing1LimitDegrees=45
  swing1Motion=Free
  swing2LimitDegrees=45
  swing2Motion=Free
  twistLimitDegrees=45
  twistMotion=Free
 }
 {bone1Name=dyn_bag_front_strap_r, bone2Name=base_bag_front_strap_r, swing1LimitDegrees=30, swing2LimitDegrees=65, twistLimitDegrees=5}
 {bone1Name=cookie_bag_01, bone2Name=cookie_bag, swing1LimitDegrees=6, swing2LimitDegrees=6, twistLimitDegrees=1}
 {
  bone1Name=cookie_bag
  bone2Name=pack_hand_l
  swing1LimitDegrees=45
  swing1Motion=Free
  swing2LimitDegrees=45
  swing2Motion=Free
  twistLimitDegrees=45
  twistMotion=Free
 }
 {
  bone1Name=pack_hand_l
  bone2Name=base_bag
  swing1LimitDegrees=45
  swing1Motion=Free
  swing2LimitDegrees=45
  swing2Motion=Free
  twistLimitDegrees=45
  twistMotion=Free
 }
]
