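# Anim Blueprint for the GallonBag back bling, a character that idles, eats cookies, skydives and reacts to damage, chest opening and puke triggers.
# Its state machine feeds Rigid Body With Control, per-bone AnimDynamics on the bag, then a gravity-overridden RigidBody.
# The EventGraph swaps physics-control parameters for frontend, cookie-eating and ear-tightening via state and anim curves.
# Puke fires after the first skydive and once after a random delay.
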AnimBlueprint /BRCosmetics/Accessories/FORT_Backpacks/Backpack_M_MED_GallonBag/Meshes/M_MED_GallonBag_Pack_AnimBP
target_skeleton=&/M_MED_GallonBag_Pack_Skeleton
graphs [
 AnimGraph {nodes [
   AnimGraphNode_StateMachine_1 = Idle and Reactive() {is=title AnimGraphNode_StateMachine}
   n1 = RigidBodyWithControl!(
    ComponentPose:Component Space Pose Link Structure=Local To Component(LocalPose=@AnimGraphNode_StateMachine_1)
    Alpha:float=1
    ExternalForce:Vector="0,0,0"
    ControlAndModifierParameters:Physics Control Control And Modifier Parameters Structure=Get Control And Modifier Parameters() -> (Control And Modifier Parameters:Physics Control Control And Modifier Parameters Structure)
    ControlAndModifierUpdates:Physics Control Control And Modifier Updates Structure=(ControlUpdates=,ControlMultiplierUpdates=,ModifierUpdates=)
    ForcesAndTorques:Rigid Body Forces And Torques Structure=(ForcesAndTorques=())
   ) -> (Pose:Component Space Pose Link Structure) {
    is=title AnimGraphNode_RigidBodyWithControl
    bEnableControls=true
    bOverrideWorldGravity=true
    evaluationResetTime=0.2
    overridePhysicsAsset=&/M_MED_GallonBag_Pack_PBA_Physics
    physicsAssetAuthoredSkeletalMesh=&/M_MED_GallonBag_Pack
    additionalSets {
     controlSetUpdates [
      {setName=Ears, names [ParentSpace_EarL, ParentSpace_EarR]}
      {setName=Arms, names [WorldSpace_ArmL, WorldSpace_ArmR]}
      {setName=FullBody, names [WorldSpace_Head, WorldSpace_Spine]}
     ]
     modifierSetUpdates [{setName=FullBody_Set, names [Spine,Head]}, {setName=Arms_Set, names [ArmL,ArmR]}, {setName=Ears_Set, names [EarL,EarR]}]
    }
    alphaBoolBlend {blendOption=Linear}
    characterSetupData {
     defaultBodyModifierData {gravityMultiplier=0, movementType=Kinematic}
     defaultParentSpaceControlData {bEnabled=false}
     defaultWorldSpaceControlData {bEnabled=false}
     limbSetupData [
      {
       bCreateBodyModifiers=true
       bCreateParentSpaceControls=true
       bCreateWorldSpaceControls=true
       bIncludeParentBone=false
       limbName=EarL
       startBone=pack_ear_l
      }
      {
       bCreateBodyModifiers=true
       bCreateParentSpaceControls=true
       bCreateWorldSpaceControls=true
       bIncludeParentBone=false
       limbName=EarR
       startBone=pack_ear_r
      }
      {
       bCreateBodyModifiers=true
       bCreateParentSpaceControls=false
       bCreateWorldSpaceControls=true
       bIncludeParentBone=false
       limbName=Head
       startBone=pack_neck_01
      }
      {
       bCreateBodyModifiers=true
       bCreateParentSpaceControls=false
       bCreateWorldSpaceControls=true
       bIncludeParentBone=false
       limbName=ArmR
       startBone=pack_clavicle_r
      }
      {
       bCreateBodyModifiers=true
       bCreateParentSpaceControls=false
       bCreateWorldSpaceControls=true
       bIncludeParentBone=false
       limbName=ArmL
       startBone=pack_clavicle_l
      }
      {
       bCreateBodyModifiers=true
       bCreateParentSpaceControls=false
       bCreateWorldSpaceControls=true
       bIncludeParentBone=true
       limbName=Spine
       startBone=pack_spine_01
      }
      {
       bCreateBodyModifiers=true
       bCreateParentSpaceControls=true
       bCreateWorldSpaceControls=false
       bIncludeParentBone=false
       limbName=Bag
       startBone=base_mid
      }
     ]
    }
    initialControlAndBodyModifierUpdates {
     controlParameters [
      Ears {data {
        angularDampingRatio=1
        angularExtraDamping=0
        angularStrength=6
        angularTargetVelocityMultiplier=1
        bDisableCollision=false
        bEnableAngularDampingRatio=true
        bEnableAngularExtraDamping=true
        bEnableAngularStrength=true
        bEnableAngularTargetVelocityMultiplier=true
        bEnableCustomControlPoint=true
        bEnableLinearDampingRatio=true
        bEnableLinearExtraDamping=true
        bEnableLinearStrength=true
        bEnableLinearTargetVelocityMultiplier=true
        bEnableMaxForce=true
        bEnableMaxTorque=true
        bEnablebDisableCollision=true
        bEnablebEnabled=true
        bEnablebOnlyControlChildObject=true
        bEnablebUseAccelerationDriveMode=true
        bEnablebUseCustomControlPoint=true
        bEnablebUseSkeletalAnimation=true
        bEnabled=true
        bOnlyControlChildObject=false
        bUseAccelerationDriveMode=true
        bUseCustomControlPoint=false
        bUseSkeletalAnimation=true
        linearDampingRatio=1
        linearExtraDamping=0
        linearStrength=0
        linearTargetVelocityMultiplier=1
        maxForce=0
        maxTorque=0
        customControlPoint=<0 0 0>
       }}
      Arms {data {
        angularDampingRatio=2
        angularExtraDamping=0
        angularStrength=15
        angularTargetVelocityMultiplier=1
        bDisableCollision=false
        bEnableAngularDampingRatio=true
        bEnableAngularExtraDamping=true
        bEnableAngularStrength=true
        bEnableAngularTargetVelocityMultiplier=true
        bEnableCustomControlPoint=true
        bEnableLinearDampingRatio=true
        bEnableLinearExtraDamping=true
        bEnableLinearStrength=true
        bEnableLinearTargetVelocityMultiplier=true
        bEnableMaxForce=true
        bEnableMaxTorque=true
        bEnablebDisableCollision=true
        bEnablebEnabled=true
        bEnablebOnlyControlChildObject=true
        bEnablebUseAccelerationDriveMode=true
        bEnablebUseCustomControlPoint=true
        bEnablebUseSkeletalAnimation=true
        bEnabled=true
        bOnlyControlChildObject=false
        bUseAccelerationDriveMode=true
        bUseCustomControlPoint=false
        bUseSkeletalAnimation=true
        linearDampingRatio=1
        linearExtraDamping=0
        linearStrength=0.5
        linearTargetVelocityMultiplier=1
        maxForce=0
        maxTorque=0
        customControlPoint=<0 0 0>
       }}
      FullBody {data {
        angularDampingRatio=1
        angularExtraDamping=0
        angularStrength=200
        angularTargetVelocityMultiplier=1
        bDisableCollision=false
        bEnableAngularDampingRatio=true
        bEnableAngularExtraDamping=true
        bEnableAngularStrength=true
        bEnableAngularTargetVelocityMultiplier=true
        bEnableCustomControlPoint=true
        bEnableLinearDampingRatio=true
        bEnableLinearExtraDamping=true
        bEnableLinearStrength=true
        bEnableLinearTargetVelocityMultiplier=true
        bEnableMaxForce=true
        bEnableMaxTorque=true
        bEnablebDisableCollision=true
        bEnablebEnabled=true
        bEnablebOnlyControlChildObject=true
        bEnablebUseAccelerationDriveMode=true
        bEnablebUseCustomControlPoint=true
        bEnablebUseSkeletalAnimation=true
        bEnabled=true
        bOnlyControlChildObject=false
        bUseAccelerationDriveMode=true
        bUseCustomControlPoint=false
        bUseSkeletalAnimation=true
        linearDampingRatio=1
        linearExtraDamping=0
        linearStrength=5
        linearTargetVelocityMultiplier=1
        maxForce=0
        maxTorque=0
        customControlPoint=<0 0 0>
       }}
      ParentSpace_Bag {data {
        angularDampingRatio=1
        angularExtraDamping=0.8
        angularStrength=5
        angularTargetVelocityMultiplier=0.5
        bDisableCollision=false
        bEnableAngularDampingRatio=true
        bEnableAngularExtraDamping=true
        bEnableAngularStrength=true
        bEnableAngularTargetVelocityMultiplier=true
        bEnableCustomControlPoint=true
        bEnableLinearDampingRatio=true
        bEnableLinearExtraDamping=true
        bEnableLinearStrength=true
        bEnableLinearTargetVelocityMultiplier=true
        bEnableMaxForce=true
        bEnableMaxTorque=true
        bEnablebDisableCollision=true
        bEnablebEnabled=true
        bEnablebOnlyControlChildObject=true
        bEnablebUseAccelerationDriveMode=true
        bEnablebUseCustomControlPoint=true
        bEnablebUseSkeletalAnimation=true
        bEnabled=true
        bOnlyControlChildObject=false
        bUseAccelerationDriveMode=true
        bUseCustomControlPoint=false
        bUseSkeletalAnimation=true
        linearDampingRatio=0.2
        linearExtraDamping=0
        linearStrength=1
        linearTargetVelocityMultiplier=0.5
        maxForce=0
        maxTorque=0
        customControlPoint=<0 0 0>
       }}
     ]
     modifierParameters [
      Ears_Set {data {
        bEnableCCD=false
        bEnableCollisionType=true
        bEnableGravityMultiplier=true
        bEnableKinematicTargetSpace=true
        bEnableMovementType=true
        bEnablePhysicsBlendWeight=true
        bEnablebEnableCCD=true
        bEnablebUpdateKinematicFromSimulation=true
        bUpdateKinematicFromSimulation=true
        collisionType=QueryAndPhysics
        gravityMultiplier=0
        kinematicTargetSpace=OffsetInBoneSpace
        movementType=Simulated
        physicsBlendWeight=1
       }}
      Arms_Set {data {
        bEnableCCD=false
        bEnableCollisionType=true
        bEnableGravityMultiplier=true
        bEnableKinematicTargetSpace=true
        bEnableMovementType=true
        bEnablePhysicsBlendWeight=true
        bEnablebEnableCCD=true
        bEnablebUpdateKinematicFromSimulation=true
        bUpdateKinematicFromSimulation=true
        collisionType=QueryAndPhysics
        gravityMultiplier=0
        kinematicTargetSpace=OffsetInBoneSpace
        movementType=Simulated
        physicsBlendWeight=1
       }}
      FullBody_Set {data {
        bEnableCCD=false
        bEnableCollisionType=true
        bEnableGravityMultiplier=true
        bEnableKinematicTargetSpace=true
        bEnableMovementType=true
        bEnablePhysicsBlendWeight=true
        bEnablebEnableCCD=true
        bEnablebUpdateKinematicFromSimulation=true
        bUpdateKinematicFromSimulation=true
        collisionType=QueryAndPhysics
        gravityMultiplier=0
        kinematicTargetSpace=OffsetInBoneSpace
        movementType=Simulated
        physicsBlendWeight=1
       }}
      Bag {data {
        bEnableCCD=false
        bEnableCollisionType=true
        bEnableGravityMultiplier=true
        bEnableKinematicTargetSpace=true
        bEnableMovementType=true
        bEnablePhysicsBlendWeight=true
        bEnablebEnableCCD=true
        bEnablebUpdateKinematicFromSimulation=true
        bUpdateKinematicFromSimulation=true
        collisionType=QueryAndPhysics
        gravityMultiplier=1
        kinematicTargetSpace=OffsetInBoneSpace
        movementType=Simulated
        physicsBlendWeight=1
       }}
     ]
    }
    overrideWorldGravity {z=-980}
    simSpaceSettings {
     maxAngularAcceleration=100
     maxAngularVelocity=7.5
     maxLinearAcceleration=100
     maxLinearVelocity=7.5
     worldAlpha=1
     externalLinearDragV=<0.2 0.2 0.2>
    }
   }
   n2 = Anim Dynamics - Bone: bag_front(ComponentPose=@n1, Alpha=1, ExternalForce="0,0,0") {
    angularDampingOverride=0.98
    angularSpringConstant=10000
    bAngularSpring=true
    bOverrideAngularDamping=true
    bUseGravityOverride=true
    gravityScale=0.35
    simulationSpace=RootRelative
    alphaBoolBlend {blendOption=Linear}
    boundBone {boneName=bag_front}
    componentLinearAccScale=<1 1 1>
    componentLinearVelScale=<1 1 1>
    gravityOverride {z=-980}
    physicsBodyDefinitions [{
      boundBone {boneName=bag_front}
      boxExtents=<20 35 35>
      constraintSetup {angularLimitsMax=<20 20 20>, angularLimitsMin=<-20 -20 -20>, angularTarget=<1 0 0>}
      localJointOffset=<4 0 -4>
     }]
   }
   n3 = Anim Dynamics - Bone: bag_l_01(ComponentPose=@n2, Alpha=1, ExternalForce="0,0,0") {
    angularDampingOverride=0.98
    angularSpringConstant=10000
    bAngularSpring=true
    bOverrideAngularDamping=true
    bUseGravityOverride=true
    gravityScale=0.35
    simulationSpace=RootRelative
    alphaBoolBlend {blendOption=Linear}
    boundBone {boneName=bag_l_01}
    componentLinearAccScale=<1 1 1>
    componentLinearVelScale=<1 1 1>
    gravityOverride {z=-980}
    physicsBodyDefinitions [{
      boundBone {boneName=bag_l_01}
      boxExtents=<20 35 35>
      constraintSetup {angularLimitsMax=<20 20 20>, angularLimitsMin=<-20 -20 -20>, angularTarget=<1 0 0>}
      localJointOffset=<4 0 -4>
     }]
   }
   n4 = Anim Dynamics - Bone: bag_l_02(ComponentPose=@n3, Alpha=1, ExternalForce="0,0,0") {
    angularDampingOverride=0.98
    angularSpringConstant=10000
    bAngularSpring=true
    bOverrideAngularDamping=true
    bUseGravityOverride=true
    gravityScale=0.35
    simulationSpace=RootRelative
    alphaBoolBlend {blendOption=Linear}
    boundBone {boneName=bag_l_02}
    componentLinearAccScale=<1 1 1>
    componentLinearVelScale=<1 1 1>
    gravityOverride {z=-980}
    physicsBodyDefinitions [{
      boundBone {boneName=bag_l_02}
      boxExtents=<20 35 35>
      constraintSetup {angularLimitsMax=<20 20 20>, angularLimitsMin=<-20 -20 -20>, angularTarget=<1 0 0>}
      localJointOffset=<4 0 -4>
     }]
   }
   n5 = Anim Dynamics - Bone: bag_r_01(ComponentPose=@n4, Alpha=1, ExternalForce="0,0,0") {
    angularDampingOverride=0.98
    angularSpringConstant=10000
    bAngularSpring=true
    bOverrideAngularDamping=true
    bUseGravityOverride=true
    gravityScale=0.35
    simulationSpace=RootRelative
    alphaBoolBlend {blendOption=Linear}
    boundBone {boneName=bag_r_01}
    componentLinearAccScale=<1 1 1>
    componentLinearVelScale=<1 1 1>
    gravityOverride {z=-980}
    physicsBodyDefinitions [{
      boundBone {boneName=bag_r_01}
      boxExtents=<20 35 35>
      constraintSetup {angularLimitsMax=<20 20 20>, angularLimitsMin=<-20 -20 -20>, angularTarget=<1 0 0>}
      localJointOffset=<4 0 -4>
     }]
   }
   n6 = Anim Dynamics - Bone: bag_r_02(ComponentPose=@n5, Alpha=1, ExternalForce="0,0,0") {
    angularDampingOverride=0.98
    angularSpringConstant=10000
    bAngularSpring=true
    bOverrideAngularDamping=true
    bUseGravityOverride=true
    gravityScale=0.35
    simulationSpace=RootRelative
    alphaBoolBlend {blendOption=Linear}
    boundBone {boneName=bag_r_02}
    componentLinearAccScale=<1 1 1>
    componentLinearVelScale=<1 1 1>
    gravityOverride {z=-980}
    physicsBodyDefinitions [{
      boundBone {boneName=bag_r_02}
      boxExtents=<20 35 35>
      constraintSetup {angularLimitsMax=<20 20 20>, angularLimitsMin=<-20 -20 -20>, angularTarget=<1 0 0>}
      localJointOffset=<4 0 -4>
     }]
   }
   n7 = RigidBody!(ComponentPose=@n6, Alpha, OverrideWorldGravity:Vector=Get Gravity Override(), ExternalForce) -> (Pose) {
    bOverrideWorldGravity=true
    overridePhysicsAsset=&/M_MED_GallonBag_Pack_Physics
    simulationSpace=BaseBoneSpace
    alphaBoolBlend {blendOption=Linear}
    baseBoneRef {boneName=root}
    componentAppliedLinearAccClamp=<3000 3000 3000>
    componentLinearAccScale=<0.3 0.5 0.4>
    componentLinearVelScale=<0.3 0.3 0.3>
    simSpaceSettings {
     maxAngularAcceleration=1000
     maxAngularVelocity=3
     maxLinearAcceleration=700
     maxLinearVelocity=500
     velocityScaleZ=0
     worldAlpha=0.275
     externalLinearDragV=<0.2 0.2>
    }
   }
   Output Pose(Result=Component To Local(ComponentPose=@n7)) {layerGroup=None}
  ]}
 Idle and Reactive {
  entry=Idle
  path=AnimGraph.AnimGraphNode_StateMachine_1.Idle and Reactive
  ids=id
  nodes [
   AnimStateNode_0 = Idle() {
    stateEntered {guid=#1}
    stateFullyBlended {guid=#2}
    stateLeft {guid=#3}
   }
   AnimStateNode_1 = Skydive Dive() {
    is=title AnimStateNode
    stateEntered {guid=#4}
    stateFullyBlended {guid=#5}
    stateLeft {guid=#6}
   }
   AnimStateNode_2 = Puke Anim() {
    is=title AnimStateNode
    stateEntered {guid=#7}
    stateFullyBlended {guid=#8}
    stateLeft {guid=#9}
   }
   AnimStateNode_3 = Reactive Damage() {
    is=title AnimStateNode
    stateEntered {guid=#10}
    stateFullyBlended {guid=#11}
    stateLeft {guid=#12}
   }
   AnimStateNode_4 = Reactive Chest Open() {
    is=title AnimStateNode
    stateEntered {guid=#13}
    stateFullyBlended {guid=#14}
    stateLeft {guid=#15}
   }
   AnimStateNode_6 = Skydive Float() {is=title AnimStateNode}
   AnimStateTransitionNode_0 = AnimStateTransitionNode() {bAutomaticRuleBasedOnSequencePlayerInState=true}
   AnimStateTransitionNode_1 = AnimStateTransitionNode() {bAutomaticRuleBasedOnSequencePlayerInState=true}
   AnimStateTransitionNode_12 = AnimStateTransitionNode() {crossfadeDuration=0}
   AnimStateTransitionNode_13 = AnimStateTransitionNode()
   AnimStateTransitionNode_14 = AnimStateTransitionNode() {bAutomaticRuleBasedOnSequencePlayerInState=true, crossfadeDuration=0.05}
   AnimStateTransitionNode_15 = AnimStateTransitionNode() {blendMode=ExpIn, crossfadeDuration=2}
   AnimStateTransitionNode_16 = AnimStateTransitionNode() {crossfadeDuration=0}
   AnimStateTransitionNode_18 = AnimStateTransitionNode()
   AnimStateTransitionNode_19 = AnimStateTransitionNode()
   AnimStateTransitionNode_20 = AnimStateTransitionNode()
   AnimStateTransitionNode_21 = AnimStateTransitionNode() {blendMode=ExpIn, crossfadeDuration=2}
   AnimStateTransitionNode_22 = AnimStateTransitionNode()
   AnimStateTransitionNode_4 = AnimStateTransitionNode() {crossfadeDuration=0}
   AnimStateTransitionNode_5 = AnimStateTransitionNode() {crossfadeDuration=0}
   AnimStateTransitionNode_6 = AnimStateTransitionNode() {
    crossfadeDuration=0
    transitionEnd {guid=#16}
    transitionInterrupt {guid=#17}
    transitionStart {guid=#18}
   }
   AnimStateTransitionNode_7 = AnimStateTransitionNode() {
    crossfadeDuration=0
    transitionEnd {guid=#19}
    transitionInterrupt {guid=#20}
    transitionStart {guid=#21}
   }
  ]
  links [
   {
    dst=Puke Anim
    rule=AnimGraph.AnimGraphNode_StateMachine_1.Idle and Reactive.AnimStateTransitionNode_18.Transition
    src=Idle
    via=AnimStateTransitionNode_18
   }
   {
    dst=Reactive Chest Open
    rule=AnimGraph.AnimGraphNode_StateMachine_1.Idle and Reactive.AnimStateTransitionNode_12.Transition
    src=Idle
    via=AnimStateTransitionNode_12
   }
   {
    dst=Reactive Damage
    rule=AnimGraph.AnimGraphNode_StateMachine_1.Idle and Reactive.AnimStateTransitionNode_16.Transition
    src=Idle
    via=AnimStateTransitionNode_16
   }
   {
    dst=Skydive Dive
    rule=AnimGraph.AnimGraphNode_StateMachine_1.Idle and Reactive.AnimStateTransitionNode_13.Transition
    src=Idle
    via=AnimStateTransitionNode_13
   }
   {
    dst=Skydive Float
    rule=AnimGraph.AnimGraphNode_StateMachine_1.Idle and Reactive.AnimStateTransitionNode_22.Transition
    src=Idle
    via=AnimStateTransitionNode_22
   }
   {
    dst=Idle
    rule=AnimGraph.AnimGraphNode_StateMachine_1.Idle and Reactive.AnimStateTransitionNode_0.Transition
    src=Puke Anim
    via=AnimStateTransitionNode_0
   }
   {
    dst=Reactive Chest Open
    rule=AnimGraph.AnimGraphNode_StateMachine_1.Idle and Reactive.AnimStateTransitionNode_7.Transition
    src=Puke Anim
    via=AnimStateTransitionNode_7
   }
   {
    dst=Reactive Damage
    rule=AnimGraph.AnimGraphNode_StateMachine_1.Idle and Reactive.AnimStateTransitionNode_6.Transition
    src=Puke Anim
    via=AnimStateTransitionNode_6
   }
   {
    dst=Idle
    rule=AnimGraph.AnimGraphNode_StateMachine_1.Idle and Reactive.AnimStateTransitionNode_14.Transition
    src=Reactive Chest Open
    via=AnimStateTransitionNode_14
   }
   {
    dst=Reactive Damage
    rule=AnimGraph.AnimGraphNode_StateMachine_1.Idle and Reactive.AnimStateTransitionNode_5.Transition
    src=Reactive Chest Open
    via=AnimStateTransitionNode_5
   }
   {
    dst=Idle
    rule=AnimGraph.AnimGraphNode_StateMachine_1.Idle and Reactive.AnimStateTransitionNode_1.Transition
    src=Reactive Damage
    via=AnimStateTransitionNode_1
   }
   {
    dst=Reactive Chest Open
    rule=AnimGraph.AnimGraphNode_StateMachine_1.Idle and Reactive.AnimStateTransitionNode_4.Transition
    src=Reactive Damage
    via=AnimStateTransitionNode_4
   }
   {
    dst=Idle
    rule=AnimGraph.AnimGraphNode_StateMachine_1.Idle and Reactive.AnimStateTransitionNode_15.Transition
    src=Skydive Dive
    via=AnimStateTransitionNode_15
   }
   {
    dst=Skydive Float
    rule=AnimGraph.AnimGraphNode_StateMachine_1.Idle and Reactive.AnimStateTransitionNode_20.Transition
    src=Skydive Dive
    via=AnimStateTransitionNode_20
   }
   {
    dst=Idle
    rule=AnimGraph.AnimGraphNode_StateMachine_1.Idle and Reactive.AnimStateTransitionNode_21.Transition
    src=Skydive Float
    via=AnimStateTransitionNode_21
   }
   {
    dst=Skydive Dive
    rule=AnimGraph.AnimGraphNode_StateMachine_1.Idle and Reactive.AnimStateTransitionNode_19.Transition
    src=Skydive Float
    via=AnimStateTransitionNode_19
   }
  ]
 }
 Idle {
  path=AnimGraph.AnimGraphNode_StateMachine_1.Idle and Reactive.AnimStateNode_0.Idle
  nodes [
   n1 = Random Sequence Player!() -> (Pose:Pose Link Structure) {
    is=title AnimGraphNode_RandomPlayer
    entries [
     {
      chanceToPlay=0.7
      maxLoopCount=0
      maxPlayRate=1
      minLoopCount=0
      minPlayRate=1
      sequence=/BRCosmetics/Accessories/FORT_Backpacks/Backpack_M_MED_GallonBag/Animation/GallonBag_Idle
      blendIn {blendOption=Linear, blendTime=0.2, customCurve=null}
     }
     {
      chanceToPlay=0.3
      maxLoopCount=0
      maxPlayRate=1
      minLoopCount=0
      minPlayRate=1
      sequence=/BRCosmetics/Accessories/FORT_Backpacks/Backpack_M_MED_GallonBag/Animation/GallonBag_IdleBreak_EatCookie
      blendIn {blendOption=Linear, blendTime=0.2, customCurve=null}
     }
    ]
   }
   Output Animation Pose(Result=@n1)
  ]
 }
 Skydive Dive {
  path=AnimGraph.AnimGraphNode_StateMachine_1.Idle and Reactive.AnimStateNode_1.Skydive Dive
  nodes [
   n1 = "Transform (Modify) Bone - Bone: pack_spine_03"(
    ComponentPose=Local To Component(LocalPose=Sequence Player 'Gallonbag_Skydive'() {playRate=1.5, sequence=/BRCosmetics/Accessories/FORT_Backpacks/Backpack_M_MED_GallonBag/Animation/Gallonbag_Skydive})
    Alpha=1
    Translation="0,-18,0"
    Rotation="0,0,-30"
    Scale="1,1,1"
   ) {
    rotationMode=BMM_Additive
    translationMode=BMM_Additive
    alphaBoolBlend {blendOption=Linear}
    boneToModify {boneName=pack_spine_03}
    rotation {roll=-30}
    translation {y=-18}
   }
   Output Animation Pose(Result=Component To Local(ComponentPose=@n1))
  ]
 }
 Puke Anim {
  path=AnimGraph.AnimGraphNode_StateMachine_1.Idle and Reactive.AnimStateNode_2.Puke Anim
  nodes [Output Animation Pose(Result=Sequence Player 'GallonBag_Reactive_Puke'() {bLoopAnimation=false, sequence=/BRCosmetics/Accessories/FORT_Backpacks/Backpack_M_MED_GallonBag/Animation/GallonBag_Reactive_Puke})]
 }
 Reactive Damage {
  path=AnimGraph.AnimGraphNode_StateMachine_1.Idle and Reactive.AnimStateNode_3.Reactive Damage
  nodes [Output Animation Pose(Result=Sequence Player 'GallonBag_Reactive_Damage'() {bLoopAnimation=false, sequence=/BRCosmetics/Accessories/FORT_Backpacks/Backpack_M_MED_GallonBag/Animation/GallonBag_Reactive_Damage})]
 }
 Reactive Chest Open {
  path=AnimGraph.AnimGraphNode_StateMachine_1.Idle and Reactive.AnimStateNode_4.Reactive Chest Open
  nodes [
   n1 = Sequence Player 'GallonBag_Reactive_ChestOpen'() {bLoopAnimation=false, sequence=/BRCosmetics/Accessories/FORT_Backpacks/Backpack_M_MED_GallonBag/Animation/GallonBag_Reactive_ChestOpen}
   Output Animation Pose(Result=@n1)
  ]
 }
 Skydive Float {
  path=AnimGraph.AnimGraphNode_StateMachine_1.Idle and Reactive.AnimStateNode_6.Skydive Float
  nodes [
   n1 = "Transform (Modify) Bone - Bone: pack_spine_03"(
    ComponentPose=Local To Component(LocalPose=Sequence Player 'Gallonbag_Skydive'() {sequence=&Gallonbag_Skydive})
    Alpha=1
    Translation="0,-5,0"
    Rotation="0,0,-15"
    Scale="1,1,1"
   ) {
    rotationMode=BMM_Additive
    scaleMode=BMM_Additive
    translationMode=BMM_Additive
    alphaBoolBlend {blendOption=Linear}
    boneToModify {boneName=pack_spine_03}
    rotation {roll=-15}
    translation {y=-5}
   }
   Output Animation Pose(Result=Component To Local(ComponentPose=@n1))
  ]
 }
 Transition {path=AnimGraph.AnimGraphNode_StateMachine_1.Idle and Reactive.AnimStateTransitionNode_0.Transition, nodes [Result()]}
 Transition {path=AnimGraph.AnimGraphNode_StateMachine_1.Idle and Reactive.AnimStateTransitionNode_1.Transition, nodes [Result()]}
 Transition {
  path=AnimGraph.AnimGraphNode_StateMachine_1.Idle and Reactive.AnimStateTransitionNode_12.Transition
  nodes [Result(bCanEnterTransition=Get Reactive Anim Chest Open() -> (Reactive Anim Chest Open:bool))]
 }
 Transition {
  path=AnimGraph.AnimGraphNode_StateMachine_1.Idle and Reactive.AnimStateTransitionNode_13.Transition
  nodes [Result(bCanEnterTransition=Get bIsSkydiveDiveMode())]
 }
 Transition {path=AnimGraph.AnimGraphNode_StateMachine_1.Idle and Reactive.AnimStateTransitionNode_14.Transition, nodes [Result()]}
 Transition {
  path=AnimGraph.AnimGraphNode_StateMachine_1.Idle and Reactive.AnimStateTransitionNode_15.Transition
  nodes [Result(bCanEnterTransition=Get bIsParachuteOpened())]
 }
 Transition {
  path=AnimGraph.AnimGraphNode_StateMachine_1.Idle and Reactive.AnimStateTransitionNode_16.Transition
  nodes [Result(bCanEnterTransition=Get Reactive Anim Damage() -> (Reactive Anim Damage:bool))]
 }
 Transition {
  path=AnimGraph.AnimGraphNode_StateMachine_1.Idle and Reactive.AnimStateTransitionNode_18.Transition
  nodes [Result(bCanEnterTransition=OR Boolean(A=Get Run Puke Anim() -> (Run Puke Anim:bool), B=Get Run Random Puke Anim() -> (Run Random Puke Anim:bool)))]
 }
 Transition {
  path=AnimGraph.AnimGraphNode_StateMachine_1.Idle and Reactive.AnimStateTransitionNode_19.Transition
  nodes [Result(bCanEnterTransition=Get bIsSkydiveDiveMode())]
 }
 Transition {
  path=AnimGraph.AnimGraphNode_StateMachine_1.Idle and Reactive.AnimStateTransitionNode_20.Transition
  nodes [Result(bCanEnterTransition=Get bIsSkydiveFloating())]
 }
 Transition {
  path=AnimGraph.AnimGraphNode_StateMachine_1.Idle and Reactive.AnimStateTransitionNode_21.Transition
  nodes [Result(bCanEnterTransition=Get bIsParachuteOpened())]
 }
 Transition {
  path=AnimGraph.AnimGraphNode_StateMachine_1.Idle and Reactive.AnimStateTransitionNode_22.Transition
  nodes [Result(bCanEnterTransition=Get bIsSkydiveFloating())]
 }
 Transition {
  path=AnimGraph.AnimGraphNode_StateMachine_1.Idle and Reactive.AnimStateTransitionNode_4.Transition
  nodes [Result(bCanEnterTransition=Get Reactive Anim Chest Open())]
 }
 Transition {
  path=AnimGraph.AnimGraphNode_StateMachine_1.Idle and Reactive.AnimStateTransitionNode_5.Transition
  nodes [Result(bCanEnterTransition=Get Reactive Anim Damage())]
 }
 Transition {
  path=AnimGraph.AnimGraphNode_StateMachine_1.Idle and Reactive.AnimStateTransitionNode_6.Transition
  nodes [Result(bCanEnterTransition=Get Reactive Anim Damage())]
 }
 Transition {
  path=AnimGraph.AnimGraphNode_StateMachine_1.Idle and Reactive.AnimStateTransitionNode_7.Transition
  nodes [Result(bCanEnterTransition=Get Reactive Anim Chest Open())]
 }
 EventGraph {nodes [
   Event BlueprintUpdateAnimation()
   > n1 = Sequence!(execute) -> (then_0, then_1, then_2:Exec, then_3:Exec)
   > n2 = Gravity Override State Function DEPRECATED(CharacterPartAnimInstance=Self-Reference(), GameTimeInSeconds=GetGameTimeInSeconds(), GravityOverrideParams=Get Gravity Override Params())
   > Set Gravity Override(Gravity Override=@n2)
   n3 = Branch(execute=@n1.then_1, Condition=Get bIsInFrontEnd())
   > Set Control And Modifier Parameters(Control And Modifier Parameters:Physics Control Control And Modifier Parameters Structure=Get Frontend Control And Modifier Parameters() -> (Frontend Control And Modifier Parameters:Physics Control Control And Modifier Parameters Structure)) -> (Output_Get:Physics Control Control And Modifier Parameters Structure)
   > Set RBWC Alpha(RBWC Alpha:double=0.33) -> (Output_Get:double)
   Set RBWC Alpha(execute=@n3.else, RBWC Alpha=1)
   > n4 = Branch(Condition="Equal (Float)"(A=GetCurveValue(CurveName=GallonBag_IdleBreak), B=1))
   > Set Control And Modifier Parameters(Control And Modifier Parameters=Get Eat Cookie Control And Modifier Parameters() -> (Eat Cookie Control And Modifier Parameters:Physics Control Control And Modifier Parameters Structure))
   n5 = Branch(execute=@n4.else, Condition="Equal (Float)"(A=GetCurveValue(CurveName=GallonBag_TightenEars), B=1))
   > Set Control And Modifier Parameters(Control And Modifier Parameters=Get Tighten Ears Control And Modifier Parameters() -> (Tighten Ears Control And Modifier Parameters:Physics Control Control And Modifier Parameters Structure))
   Set Control And Modifier Parameters(
    execute=@n5.else
    Control And Modifier Parameters=Get Default Control And Modifier Parameters() -> (Default Control And Modifier Parameters:Physics Control Control And Modifier Parameters Structure)
   )
   Do Once(execute=@n1.then_2)
   > Delay(Duration=RandomFloatInRange(Min=300, Max=360))
   > Set Run Random Puke Anim(Run Random Puke Anim:bool=true) -> (Output_Get:bool)
   n6 = Event BlueprintProcessAnimInputEvent()
   > n7 = Sequence!(execute) -> (then_0, then_1)
   > Branch(Condition=OR Boolean(
     A=IsNamed(self=@n6.AnimInputEvent, InEventName=aie.PlayerVitality.ShieldsDecreased)
     B=IsNamed(self=@n6.AnimInputEvent, InEventName=aie.PlayerVitality.HealthDecreased)
    ))
   > Set Reactive Anim Damage(Reactive Anim Damage:bool=true) -> (Output_Get:bool)
   > RetriggerableDelay(Duration=0.2)
   > Set Reactive Anim Damage(Reactive Anim Damage=false)
   Branch(execute=@n7.then_1, Condition=IsNamed(self=@n6.AnimInputEvent, InEventName=aie.ChestOpened))
   > Set Reactive Anim Chest Open(Reactive Anim Chest Open:bool=true) -> (Output_Get:bool)
   > RetriggerableDelay(Duration=0.2)
   > Set Reactive Anim Chest Open(Reactive Anim Chest Open=false)
   Event OnEndSkydiving()
   > n8 = Branch(Condition="Equal (Integer)"(A=Get Puke Anim Event() -> (Puke Anim Event:int=0)))
   > Set Run Puke Anim(Run Puke Anim:bool=true) -> (Output_Get:bool)
   > Set Puke Anim Event(Puke Anim Event:int=1) -> (Output_Get:int=0)
   Set Run Puke Anim(execute=@n8.else, Run Puke Anim=false)
   Event BlueprintPostEvaluateAnimation() {is=title K2Node_Event}
   > Set Run Puke Anim(Run Puke Anim=false)
   > Set Run Random Puke Anim(Run Random Puke Anim=false)
   TryGetPawnOwner()
  ]}
]
variables [
 Control And Modifier Parameters {
  type=PhysicsControlControlAndModifierParameters
  default {
   controlMultiplierParameters []
   controlParameters []
   modifierParameters [None {data {
      bEnableCCD=false
      bEnableCollisionType=true
      bEnableGravityMultiplier=true
      bEnableKinematicTargetSpace=true
      bEnableMovementType=true
      bEnablePhysicsBlendWeight=true
      bEnablebEnableCCD=true
      bEnablebUpdateKinematicFromSimulation=true
      bUpdateKinematicFromSimulation=true
      collisionType=QueryAndPhysics
      gravityMultiplier=1
      kinematicTargetSpace=OffsetInBoneSpace
      movementType=Simulated
      physicsBlendWeight=1
     }}]
  }
 }
 Default Control And Modifier Parameters {type=PhysicsControlControlAndModifierParameters, default {controlMultiplierParameters [], controlParameters [], modifierParameters []}}
 Eat Cookie Control And Modifier Parameters {
  type=PhysicsControlControlAndModifierParameters
  default {
   controlMultiplierParameters []
   controlParameters [WorldSpace_ArmR {data {
      angularDampingRatio=1
      angularExtraDamping=0
      angularStrength=20000
      angularTargetVelocityMultiplier=1
      bDisableCollision=false
      bEnableAngularDampingRatio=true
      bEnableAngularExtraDamping=true
      bEnableAngularStrength=true
      bEnableAngularTargetVelocityMultiplier=true
      bEnableCustomControlPoint=true
      bEnableLinearDampingRatio=true
      bEnableLinearExtraDamping=true
      bEnableLinearStrength=true
      bEnableLinearTargetVelocityMultiplier=true
      bEnableMaxForce=true
      bEnableMaxTorque=true
      bEnablebDisableCollision=true
      bEnablebEnabled=true
      bEnablebOnlyControlChildObject=true
      bEnablebUseAccelerationDriveMode=true
      bEnablebUseCustomControlPoint=true
      bEnablebUseSkeletalAnimation=true
      bEnabled=true
      bOnlyControlChildObject=true
      bUseAccelerationDriveMode=true
      bUseCustomControlPoint=false
      bUseSkeletalAnimation=true
      linearDampingRatio=1
      linearExtraDamping=0
      linearStrength=20000
      linearTargetVelocityMultiplier=1
      maxForce=0
      maxTorque=0
      customControlPoint=<0 0 0>
     }}]
   modifierParameters [ArmR {data {
      bEnableCCD=false
      bEnableCollisionType=true
      bEnableGravityMultiplier=true
      bEnableKinematicTargetSpace=true
      bEnableMovementType=true
      bEnablePhysicsBlendWeight=true
      bEnablebEnableCCD=true
      bEnablebUpdateKinematicFromSimulation=true
      bUpdateKinematicFromSimulation=true
      collisionType=QueryAndPhysics
      gravityMultiplier=0
      kinematicTargetSpace=OffsetInBoneSpace
      movementType=Kinematic
      physicsBlendWeight=1
     }}]
  }
 }
 Frontend Control And Modifier Parameters {
  type=PhysicsControlControlAndModifierParameters
  default {
   controlMultiplierParameters []
   controlParameters [FullBody {data {
      angularDampingRatio=1
      angularExtraDamping=0.5
      angularStrength=100
      angularTargetVelocityMultiplier=1
      bDisableCollision=false
      bEnableAngularDampingRatio=true
      bEnableAngularExtraDamping=true
      bEnableAngularStrength=true
      bEnableAngularTargetVelocityMultiplier=true
      bEnableCustomControlPoint=true
      bEnableLinearDampingRatio=true
      bEnableLinearExtraDamping=true
      bEnableLinearStrength=true
      bEnableLinearTargetVelocityMultiplier=true
      bEnableMaxForce=true
      bEnableMaxTorque=true
      bEnablebDisableCollision=true
      bEnablebEnabled=true
      bEnablebOnlyControlChildObject=true
      bEnablebUseAccelerationDriveMode=true
      bEnablebUseCustomControlPoint=true
      bEnablebUseSkeletalAnimation=true
      bEnabled=true
      bOnlyControlChildObject=false
      bUseAccelerationDriveMode=true
      bUseCustomControlPoint=false
      bUseSkeletalAnimation=true
      linearDampingRatio=1
      linearExtraDamping=0.5
      linearStrength=1
      linearTargetVelocityMultiplier=1
      maxForce=0
      maxTorque=0
      customControlPoint=<0 0 0>
     }}]
   modifierParameters []
  }
 }
 Gravity Override {type=Vector}
 Gravity Override Ears {type=Vector}
 Gravity Override Params {type=GravityOverrideParamsStruct}
 Gravity Override Params Ears {
  type=GravityOverrideParamsStruct
  default {
   skydive_DivingDown {jointName=root, gravityOverride=<0 -700 600>, windAmplitude=<200 -500 400>, windFrequency=<10 20 8>, windOffset=<0 -600 400>}
   skydive_FloatingDown {gravityOverride=<0 -450 300>, windAmplitude=<100 -250 200>, windFrequency=<15 20 15>, windOffset=<0 -300 200>}
   skydive_Parachuting {gravityOverride=<0 -400 -400>, windAmplitude=<100 -250 200>, windFrequency=<10 15 10>}
  }
 }
 Puke Anim Event {type=integer}
 Random Puke Anim Event {type=integer}
 RBWC Alpha {type=number}
 Reactive Anim Chest Open {type=boolean}
 Reactive Anim Damage {type=boolean}
 Run Puke Anim {type=boolean}
 Run Random Puke Anim {type=boolean}
 Tighten Ears Control And Modifier Parameters {
  type=PhysicsControlControlAndModifierParameters
  default {
   controlMultiplierParameters []
   controlParameters [Ears {data {
      angularDampingRatio=1
      angularExtraDamping=0
      angularStrength=20000
      angularTargetVelocityMultiplier=1
      bDisableCollision=false
      bEnableAngularDampingRatio=true
      bEnableAngularExtraDamping=true
      bEnableAngularStrength=true
      bEnableAngularTargetVelocityMultiplier=true
      bEnableCustomControlPoint=true
      bEnableLinearDampingRatio=true
      bEnableLinearExtraDamping=true
      bEnableLinearStrength=true
      bEnableLinearTargetVelocityMultiplier=true
      bEnableMaxForce=true
      bEnableMaxTorque=true
      bEnablebDisableCollision=true
      bEnablebEnabled=true
      bEnablebOnlyControlChildObject=true
      bEnablebUseAccelerationDriveMode=true
      bEnablebUseCustomControlPoint=true
      bEnablebUseSkeletalAnimation=true
      bEnabled=true
      bOnlyControlChildObject=true
      bUseAccelerationDriveMode=true
      bUseCustomControlPoint=false
      bUseSkeletalAnimation=true
      linearDampingRatio=1
      linearExtraDamping=0
      linearStrength=20000
      linearTargetVelocityMultiplier=1
      maxForce=0
      maxTorque=0
      customControlPoint=<0 0 0>
     }}]
   modifierParameters []
  }
 }
]
